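# Cheatbox is a Default Lit material in the gameplay cheat inventory effects folder.
# Its graph contains no expressions, so every surface input stays at its default.
# It assigns the Purple_Flesh physical material and uses the legacy pixel depth offset mode.
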Material /Game/Effects/Fort_Effects/Gameplay/CheatInventory/cheatbox
graph {
 expressions {}
 properties {physMaterial=/Game/FortressPhysicalMaterials/Purple_Flesh, pixelDepthOffsetMode=PDOM_Legacy, usedShadingModels=Default Lit}
}
